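# Unlit eye material for the Dadbro enemy character, set up for skeletal meshes and mesh deformers.
# Its only node is an emissive color vector parameter with an HDR orange default bright enough for a glow, with no textures or other math.
# It is assigned the Purple_Flesh physical material.
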Material /Game/Characters/Enemies/DUDEBRO/Materials/M_Dadbro_Eyes
graph {
 properties {
  bUsedWithMeshDeformer=true
  bUsedWithSkeletalMesh=true
  physMaterial=/Game/FortressPhysicalMaterials/Purple_Flesh
  pixelDepthOffsetMode=PDOM_Legacy
  shadingModel=MSM_Unlit
  usedShadingModels=Unlit
 }
 nodes [VectorParameter() {parameterName=EmissiveColor-G3, defaultValue=<rgba 48 6.3366 0 1>}]
}
