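# Material function that layers a travelling light-wave and pulsing full-body glow onto the Dudebro (Dadbro) enemy's skin attributes.
# World-space sphere masks, broken up by panning noise and bump-offset turbulence, drive curve-atlas ramps that darken BaseColor and add emissive.
# Fresnel and AO shape the body glow; optional texture masks or view-angle falloff weight the emissive.
# A static switch sources wave and glow controls from MPC_Dudebro or material parameters and object position.
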MaterialFunction /Game/Characters/Enemies/DUDEBRO/FX/SkinFX/MF_DadbroSkinFX
graph {nodes [
  n1 = GetMaterialAttributes(None=FunctionInput() {inputName=InMA, inputType=FunctionInput_MaterialAttributes}) {attributeGetTypes [#1, #2]}
  n2 = StaticSwitchParameter(
   True=CollectionParameter() {collection=&/MPC_Dudebro, parameterName=BP_WaveOffset}
   False=ScalarParameter() {parameterName=WaveOffset}
  ) {defaultValue=true, parameterName=UseMaterialParamCollection}
  n3 = CollectionParameter() {collection=&MPC_Dudebro, parameterName=BP_WaveCenter}
  n4 = ObjectPositionWS()
  n5 = TextureCoordinate()
  n6 = ScalarParameter() {defaultValue=2, parameterName=UV}
  n7 = Time()
  n8 = TextureSample(UVs=Add(A=Multiply(A=@n5, B=@n6), B=Multiply(A=@n7) {constB=0.02})) {samplerType=SAMPLERTYPE_LinearColor, texture=&T_NoisePack_LinearColor}
  n9 = TextureCoordinate()
  n10 = ScalarParameter() {defaultValue=2, parameterName=UV2}
  n11 = TextureSample(UVs=BumpOffset(Coordinate=Multiply(A=@n9, B=@n10), Height=LinearInterpolate(Alpha=@n8.G) {constA=0.5}, HeightRatioInput=Constant() {r=0.5})) {texture=/Game/Global/Textures/Default/Caustics/T_Turbulence_Seamless}
  n12 = LinearInterpolate(Alpha=Subtract(A=@n8.R, B=@n11.R)) {constA=0.8}
  n13 = Constant() {r=2000}
  n14 = StaticSwitchParameter(
   True=CollectionParameter() {collection=&MPC_Dudebro, parameterName=BP_WaveRadius}
   False=ScalarParameter() {defaultValue=1, parameterName=WaveRadius}
  ) {defaultValue=true, parameterName=UseMaterialParamCollection}
  n15 = SphereMask(
   A=WorldPosition().XYZ
   B=ComponentMask(None=StaticSwitchParameter(True=@n3, False=@n4) {defaultValue=true, parameterName=UseMaterialParamCollection}) {b=true, g=true, r=true}
   Radius=Multiply(A=@n12, B=Multiply(A=@n13, B=@n14))
   Hardness=Constant()
  )
  n16 = CurveAtlasRowParameter(CurveTime=Add(A=@n15, B=LinearInterpolate(Alpha=@n2) {constA=-1})) {
   atlas=&AthenaCurveAtlas01
   curve=/Game/Athena/LinearColorCurves/Curve_StormKingSkinFXRamp4
   defaultValue=32
   parameterName=BP_Lightwave_RampCurve
  }
  n17 = CollectionParameter() {collection=&MPC_Dudebro, parameterName=BP_FullBodyGlowCenter}
  n18 = ObjectPositionWS()
  n19 = SphereMask(
   A=WorldPosition().XYZ
   B=ComponentMask(None=StaticSwitchParameter(True=@n17, False=@n18) {defaultValue=true, parameterName=UseMaterialParamCollection}) {b=true, g=true, r=true}
   Radius=Multiply(A=@n12, B=ScalarParameter() {defaultValue=6000, parameterName=Param_2})
   Hardness=Constant()
  )
  n20 = Add(A=@n19, B=Multiply(A=Time(), B=Constant() {r=0.2}))
  n21 = OneMinus(None=Abs(None=Sine(None=@n20)))
  n22 = Power(Base=@n19, Exp=ScalarParameter() {defaultValue=5.5, sortPriority=3}) {constExponent=3}
  n23 = ScalarParameter() {defaultValue=1.5, parameterName=Param_1, sortPriority=3}
  n24 = Multiply(
   A=Multiply(A=LinearInterpolate(Alpha=@n21) {constA=0.5}, B=Multiply(A=@n22, B=@n23) {constB=3})
   B=StaticSwitchParameter(
    True=CollectionParameter() {collection=&MPC_Dudebro, parameterName=BP_FullBodyGlow}
    False=ScalarParameter() {defaultValue=0.75, parameterName=FullBodyGlow}
   ) {defaultValue=true, parameterName=UseMaterialParamCollection}
  )
  n25 = Saturate(None=Add(A=@n15, B=@n24))
  n26 = Multiply(A=@n25, B=@n25)
  n27 = Transform(None=@n1.Normal)
  n28 = DotProduct(A=@n27, B=CameraVectorWS())
  n29 = Multiply(A=@n28, B=@n28)
  n30 = LinearInterpolate(Alpha=SquareRoot(None=@n11.R)) {constA=0.4, constB=1.5}
  n31 = Multiply(A=Multiply(A=@n16.R, B=LinearInterpolate(A=@n26, B=@n25, Alpha=@n29) {constA=0.2}), B=@n30)
  n32 = Saturate(None=@n31)
  n33 = GetMaterialAttributes(None=SetMaterialAttributes!(MaterialAttributes=@n1.MaterialAttributes, BaseColor=LinearInterpolate(A=@n1.BaseColor, Alpha=@n32) {constB=0}) {attributeSetTypes [#2]}) {attributeGetTypes [#3]}
  n34 = TextureSampleParameter2D() {
   desc=Texture Masks
   parameterName=Texture Masks
   samplerType=SAMPLERTYPE_Masks
   texture=/Game/Characters/Enemies/DUDEBRO/Textures/T_DUDEBRO_Body_Body_M
  }
  n35 = OneMinus(None=@n34.G)
  n36 = Multiply(A=@n35, B=@n35)
  n37 = OneMinus(None=@n28)
  n38 = FunctionInput(Preview=StaticBool()) {bUsePreviewValueAsDefault=true, inputName="FresnelRevert(WithOutMask)", inputType=FunctionInput_StaticBool, sortPriority=2}
  n39 = StaticBool() {value=true}
  n40 = CurveAtlasRowParameter(CurveTime=@n31) {
   atlas=&AthenaCurveAtlas01
   curve=/Game/Athena/LinearColorCurves/Curve_StormKingSkinFXColor5
   defaultValue=33
   parameterName=ColorCurve
  }
  n41 = Multiply(
   A=MaterialFunctionCall(Hue Shift Percentage=ScalarParameter() {parameterName=ColorHueShift}, Texture=@n40) {materialFunction=&HueShift}.Result
   B=StaticSwitchParameter(
    True=CollectionParameter() {collection=&MPC_Dudebro, parameterName=BP_WaveIntensity}
    False=ScalarParameter() {defaultValue=1, parameterName=WaveIntensity}
   ) {defaultValue=true, parameterName=UseMaterialParamCollection}
  )
  n42 = CameraVectorWS()
  n43 = Subtract(A=OneMinus(None=DotProduct(A=@n42, B=@n27)), B=ScalarParameter() {parameterName=FresnelFalloff})
  n44 = Multiply(
   A=LinearInterpolate(A=@n34.R, B=OneMinus(None=@n34.R), Alpha=ScalarParameter() {parameterName=FlipAO})
   B=ScalarParameter() {defaultValue=0.1, parameterName=AO_Strength}
  )
  n45 = Multiply(A=@n24, B=Clamp(None=Subtract(A=@n43, B=@n44)))
  n46 = CurveAtlasRowParameter(CurveTime=@n45) {
   atlas=&AthenaCurveAtlas01
   curve=/Game/Athena/LinearColorCurves/Curve_StormKingSkinFXWeakColor
   defaultValue=140
   parameterName=ColorCurve
  }
  n47 = CollectionParameter() {collection=&MPC_Dudebro, parameterName=BP_FullBodyGlow}
  n48 = ScalarParameter() {defaultValue=0.75, parameterName=FullBodyGlow}
  n49 = Multiply(
   A=@n46
   B=Multiply(
    A=ScalarParameter() {defaultValue=40, parameterName=WeakGlowStrength}
    B=StaticSwitchParameter(True=@n47, False=@n48) {defaultValue=true, parameterName=UseMaterialParamCollection}
   )
  )
  n50 = Constant() {r=1}
  n51 = Multiply(
   A=StaticSwitch(
    True=LinearInterpolate(A=@n36) {constAlpha=0.2, constB=0.2}
    False=StaticSwitch(True=@n29, False=@n37, Value=@n38)
    Value=FunctionInput(Preview=@n39) {bUsePreviewValueAsDefault=true, inputName=HaveMasks, inputType=FunctionInput_StaticBool, sortPriority=1}
   )
   B=Multiply(
    A=Add(A=@n41, B=@n49)
    B=FunctionInput(Preview=@n50) {bUsePreviewValueAsDefault=true, inputName=EmissiveMutiply, inputType=FunctionInput_Scalar, sortPriority=3}
   )
  )
  FunctionOutput(None=SetMaterialAttributes!(MaterialAttributes=@n33.MaterialAttributes, EmissiveColor=Add(A=@n33.EmissiveColor, B=@n51)) {attributeSetTypes [#3]}) {outputName=MA}
  MaterialFunctionCall(Hue Shift Percentage=ScalarParameter() {parameterName=ColorHueShift}, Texture=@n46) {materialFunction=&HueShift}
  Multiply(A=@n24, B=@n30)
  n52 = Add(A=ComponentMask(None=WorldPosition().XYZ) {b=true}, B=ScalarParameter() {defaultValue=816, parameterName=WorldZcut})
  Multiply(A=@n51, B=Saturate(None=Divide(A=@n52) {constB=750}))
  Multiply(A=@n45, B=VectorParameter() {parameterName=WEakGLowStrengtrh, defaultValue=<rgba 1 0.0561 0 1>}.RGB)
  Power(Base=@n49)
 ]}
