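# Main subsurface-profile skin material for the DUDEBRO (Dadbro) enemy, with glowing weakpoints on its body.
# Texture maps drive tinted diffuse, packed spec/metal/roughness, normal and subsurface, with a mask-scaled fresnel rim added to base color.
# IsTail and IsArms switches select emissive: back/sides, arms, or chest/mouth/eyes, using per-weakpoint scalars and a sharp U-coordinate left/right split.
# The resulting material attributes pass through MF_DadbroSkinFX.
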Material /Game/Characters/Enemies/DUDEBRO/Materials/M_Dadbro_Main
graph {
 properties {
  bUseMaterialAttributes=true
  bUsedWithMeshDeformer=true
  bUsedWithSkeletalMesh=true
  physMaterial=/Game/FortressPhysicalMaterials/Purple_Flesh
  pixelDepthOffsetMode=PDOM_Legacy
  shadingModel=MSM_SubsurfaceProfile
  subsurfaceProfile=/Game/Packages/Fortress_SharedMaterials/SubSurfaceProfile/SS_MistMonster_012
  usedShadingModels=Subsurface Profile
 }
 nodes [
  n1 = TextureSampleParameter2D() {
   desc=Normal Map
   group=Base Material
   parameterName=Normal Map
   samplerType=SAMPLERTYPE_Normal
   texture=/Game/Characters/Enemies/DUDEBRO/Textures/T_DUDEBRO_Body_Body_N
  }
  n2 = ScalarParameter() {defaultValue=0.5, group=Base Material, parameterName=Specular Scale}
  n3 = TextureSampleParameter2D() {
   desc=R: Spec G: Metal B: Roughness
   group=Base Material
   parameterName=Specular Masks
   samplerType=SAMPLERTYPE_Masks
   texture=/Game/Characters/Enemies/DUDEBRO/Textures/T_DUDEBRO_Body_Body_S
  }
  n4 = ScalarParameter() {group=Base Material, parameterName=RoughnessMin}
  n5 = ScalarParameter() {defaultValue=1, group=Base Material, parameterName=RoughnessMax}
  n6 = TextureSampleParameter2D() {
   desc=Diffuse Texture
   group=Base Material
   parameterName=Diffuse Map
   texture=/Game/Characters/Enemies/DUDEBRO/Textures/T_DUDEBRO_Body_Body_D
  }
  n7 = TextureSampleParameter2D() {
   desc=Texture Masks
   group=Base Material
   parameterName=Texture Masks
   samplerType=SAMPLERTYPE_Masks
   texture=/Game/Characters/Enemies/DUDEBRO/Textures/T_DUDEBRO_Body_Body_M
  }
  n8 = ScalarParameter() {defaultValue=0.1, desc=Fresnel Brightness, group=Base Material, parameterName=Fresnel Brightness}
  n9 = Fresnel(ExponentIn=Constant() {r=3}, Normal=VertexNormalWS())
  n10 = Add(
   A=Multiply(A=VectorParameter() {group=Base Material, parameterName=Diffuse_Tint, defaultValue=<rgba 1 1 1 1>}.RGB, B=@n6.RGB)
   B=Multiply(A=Multiply(A=@n7.R, B=@n7.R) {desc=ambient occlusion}, B=Multiply(A=@n8, B=@n9))
  )
  n11 = VectorParameter() {parameterName=Weakpoint-EmissiveColor, defaultValue=<rgba 1 0.5387 0 1>}
  n12 = TextureSample() {texture=/Game/Characters/Enemies/DUDEBRO/Textures/T_DudeBro_GlowMask}
  n13 = ComponentMask(None=TextureCoordinate()) {r=true}
  n14 = Constant() {r=-100}
  n15 = Saturate(None=Add(A=Multiply(A=@n13, B=@n14), B=Constant() {r=50}))
  n16 = OneMinus(None=@n15)
  n17 = Add(
   A=Multiply(A=ScalarParameter() {desc=Arm-R, parameterName=D7}, B=Multiply(A=@n12.R, B=@n15))
   B=Multiply(A=ScalarParameter() {desc=Arm-L, parameterName=D8}, B=Multiply(A=@n12.R, B=@n16))
  )
  n18 = TextureSample() {samplerType=SAMPLERTYPE_LinearColor, texture=/Game/Characters/Enemies/DUDEBRO/Textures/T_DUDEBRO_Body_Body_FX}
  n19 = Multiply(A=Multiply(A=ScalarParameter() {desc=Chest, parameterName=D3}, B=@n18.B), B=@n11.RGB)
  n20 = Add(
   A=Multiply(A=@n18.R, B=VectorParameter() {parameterName=MouthEmissiveColor, defaultValue=<rgba 0 0 0 1>}.RGB)
   B=Multiply(A=@n18.G, B=VectorParameter() {parameterName=EyesEmissiveColor, defaultValue=<rgba 1 0.3354 0.0315 1>}.RGB)
  )
  n21 = Multiply(A=ScalarParameter() {desc=Back, parameterName=D4}, B=@n12.B)
  n22 = Add(
   A=Multiply(A=ScalarParameter() {desc=Side-R, parameterName=D5}, B=Multiply(A=@n12.G, B=@n15))
   B=Multiply(A=ScalarParameter() {desc=Side-L, parameterName=D6}, B=Multiply(A=@n12.G, B=@n16))
  )
  n23 = StaticSwitchParameter(
   True=Multiply(A=Add(A=@n21, B=@n22), B=@n11.RGB)
   False=StaticSwitchParameter(True=Multiply(A=@n11.RGB, B=@n17), False=Add(A=@n19, B=@n20)) {parameterName=IsArms}
  ) {parameterName=IsTail}
  n24 = Multiply(A=ScalarParameter() {defaultValue=1, group=Base Material, parameterName=SSS Opacity}, B=@n7.B)
  n25 = Multiply(A=@n6.RGB, B=VectorParameter() {group=Base Material, parameterName=SSS Color, defaultValue=<rgba 1 1 1 1>}.RGB)
  MaterialFunctionCall!(
   InMA=MakeMaterialAttributes!(
    BaseColor=@n10
    Metallic=@n3.G
    Specular=Multiply(A=@n2, B=@n3.R)
    Roughness=LinearInterpolate(A=@n4, B=@n5, Alpha=@n3.B)
    Anisotropy
    EmissiveColor=@n23
    Opacity=@n24
    OpacityMask
    Normal=@n1.RGB
    Tangent
    WorldPositionOffset
    SubsurfaceColor=Multiply(A=@n25, B=@n24)
    ClearCoat
    ClearCoatRoughness
    AmbientOcclusion
    Refraction
    CustomizedUVs_0
    CustomizedUVs_1
    CustomizedUVs_2
    CustomizedUVs_3
    CustomizedUVs_4
    CustomizedUVs_5
    CustomizedUVs_6
    CustomizedUVs_7
    PixelDepthOffset
    ShadingModel
    Displacement
   )
   HaveMasks
   "FresnelRevert(WithOutMask)"
   EmissiveMutiply
  ) {materialFunction=/Game/Characters/Enemies/DUDEBRO/FX/SkinFX/MF_DadbroSkinFX}
  Multiply(A=@n10, B=@n23)
 ]
}
