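# Unlit material for the eyes of the Dudebro enemy, usable on skeletal meshes and with mesh deformers.
# Its graph holds a single vector parameter, EmissiveColor-G3, with a warm, overbright default color; no link to a material output is shown.
# The Purple_Flesh physical material is assigned to it.
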Material /Game/Characters/Enemies/DUDEBRO/Materials/M_Dudebro_Eyes
graph {
 properties {
  bUsedWithMeshDeformer=true
  bUsedWithSkeletalMesh=true
  physMaterial=/Game/FortressPhysicalMaterials/Purple_Flesh
  pixelDepthOffsetMode=PDOM_Legacy
  shadingModel=MSM_Unlit
  usedShadingModels=Unlit
 }
 nodes [VectorParameter() {parameterName=EmissiveColor-G3, defaultValue=<rgba 3 2.3118 1.5 1>}]
}
