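# Masked, two-sided horn material for the DUDEBRO enemy, with per-segment hiding and independent glow for left and right horns.
# A hard UV.x split picks side parameters; per-socket scalars weight two death-mask textures into opacity.
# V gradients and tiled crack textures shape emissive glow and roughness, while side glow also tints base color.
# MF_DudebroSkinFX processes the attributes; the BuffetBro switch then masks emissive by side via HitReactLR.
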Material /Game/Characters/Enemies/DUDEBRO/Materials/M_Dudebro_Horns
graph {
 properties {
  bUseMaterialAttributes=true
  bUsedWithMeshDeformer=true
  bUsedWithSkeletalMesh=true
  blendMode=BLEND_Masked
  physMaterial=/Game/FortressPhysicalMaterials/Stone
  pixelDepthOffsetMode=PDOM_Legacy
  twoSided=true
  usedShadingModels=Default Lit
 }
 nodes [
  n1 = TextureSampleParameter2D() {
   desc=Texture Masks
   parameterName=Texture Masks
   samplerType=SAMPLERTYPE_Masks
   texture=/Game/Characters/Enemies/DUDEBRO/Textures/T_DUDEBRO_Horns_M
  }
  n2 = Power(Base=OneMinus(None=@n1.R), Exp=ScalarParameter() {defaultValue=2, parameterName=tt32})
  n3 = Multiply(A=Time(), B=ScalarParameter() {defaultValue=1.5, parameterName=SpeedA})
  n4 = Multiply(A=ScalarParameter() {defaultValue=1, parameterName=SpeedA11}, B=ConstantBiasScale(None=Sine(None=@n3)))
  Add(A=@n2, B=Multiply(A=@n2, B=@n4))
  Constant3Vector() {constant=<rgba 6 6 0 0>}
  Constant() {r=0.5}
  Constant() {r=0.5}
  Constant() {r=0.5}
  Multiply()
  n5 = ScalarParameter() {defaultValue=0.75, parameterName=xvbx43}
  n6 = TextureCoordinate()
  n7 = ComponentMask(None=@n6) {g=true}
  n8 = OneMinus(None=@n7)
  n9 = Power(Base=@n8, Exp=ScalarParameter() {defaultValue=13, parameterName=bcbfhftr})
  Multiply(A=Saturate(None=Add(A=@n4, B=@n5)), B=@n9)
  n10 = TextureSample() {samplerType=SAMPLERTYPE_Masks, texture=/Game/Characters/Enemies/DUDEBRO/Textures/T_DudeBro_HornDeathMask2}
  n11 = Multiply(A=ScalarParameter() {defaultValue=1, desc=G, parameterName=Socket-RHorn-05}, B=@n10.G)
  n12 = ScalarParameter() {defaultValue=1, desc=B, parameterName=Socket-RHorn-03}
  n13 = TextureSample() {samplerType=SAMPLERTYPE_Masks, texture=/Game/Characters/Enemies/DUDEBRO/Textures/T_DudeBro_HornDeathMask}
  n14 = Multiply(A=ScalarParameter() {defaultValue=1, desc=G, parameterName=Socket-RHorn-02}, B=@n13.G)
  n15 = Multiply(A=ScalarParameter() {defaultValue=1, desc=R, parameterName=Socket-RHorn-01}, B=@n13.R)
  n16 = Add(
   A=Multiply(A=ScalarParameter() {defaultValue=1, desc=R, parameterName=Socket-RHorn-04}, B=@n10.R)
   B=Add(A=Multiply(A=@n12, B=@n13.B), B=Add(A=@n14, B=@n15))
  )
  n17 = ComponentMask(None=@n6) {r=true}
  n18 = Constant() {r=-100}
  n19 = Saturate(None=Add(A=Multiply(A=@n17, B=@n18), B=Constant() {r=50}))
  n20 = Multiply(A=ScalarParameter() {defaultValue=1, parameterName=Socket-LHorn-05}, B=@n10.G)
  n21 = ScalarParameter() {defaultValue=1, parameterName=Socket-LHorn-03}
  n22 = Multiply(A=ScalarParameter() {defaultValue=1, parameterName=Socket-LHorn-02}, B=@n13.G)
  n23 = Multiply(A=ScalarParameter() {defaultValue=1, parameterName=Socket-LHorn-01}, B=@n13.R)
  n24 = Add(
   A=Multiply(A=ScalarParameter() {defaultValue=1, parameterName=Socket-LHorn-04}, B=@n10.R)
   B=Add(A=Multiply(A=@n21, B=@n13.B), B=Add(A=@n22, B=@n23))
  )
  n25 = LinearInterpolate(A=Add(A=Add(A=@n20, B=@n24), B=@n10.B) {desc=B-Add stubby horn}, B=Add(A=@n10.B, B=Add(A=@n11, B=@n16)) {desc=B-Add stubby horn}, Alpha=@n19) {desc="A=Left\r\nB=Right"}
  n26 = Add(A=Add(A=@n10.R, B=@n10.R), B=@n10.B)
  n27 = Add(A=Add(A=@n13.R, B=@n13.R), B=@n13.B)
  n28 = Saturate(None=OneMinus(None=Add(A=@n27, B=@n26)))
  n29 = TextureSampleParameter2D() {
   desc=R: Spec G: Metal B: Roughness
   group=Base Material
   parameterName=Specular Masks
   samplerType=SAMPLERTYPE_Masks
   texture=/Game/Characters/Enemies/DUDEBRO/Textures/T_DUDEBRO_Horns_S
  }
  n30 = TextureCoordinate()
  n31 = VectorParameter() {parameterName=UV, defaultValue=<rgba 6 6 0 0>}
  n32 = TextureSample(UVs=Multiply(A=@n30, B=AppendVector(A=@n31.R, B=@n31.R))) {texture=/Game/Effects/Fort_Effects/Textures/Tile/T_Glass_Cracks_M}
  n33 = TextureSample(UVs=Multiply(A=@n30, B=ScalarParameter() {defaultValue=14, parameterName=gogog})) {texture=&T_Glass_Cracks_M}
  n34 = Add(A=@n33.R, B=@n33.R)
  n35 = Add(A=Multiply(A=@n8, B=@n32.G), B=@n34)
  n36 = LinearInterpolate(
   A=ScalarParameter() {group=Base Material, parameterName=RoughnessMin}
   B=ScalarParameter() {defaultValue=1, group=Base Material, parameterName=RoughnessMax}
   Alpha=@n29.B
  )
  n37 = ScalarParameter() {defaultValue=3, parameterName=6yhgrt}
  n38 = ScalarParameter() {defaultValue=5, parameterName=vxvdsdfswww}
  n39 = Add(A=Multiply(A=@n2, B=Power(Base=@n8, Exp=@n37)), B=Multiply(A=Power(Base=@n8, Exp=@n38), B=@n32.G))
  n40 = Multiply(
   A=Power(Base=@n8, Exp=ScalarParameter() {defaultValue=10, parameterName=mdrw3523w})
   B=ScalarParameter() {defaultValue=10, parameterName=MM1}
  )
  n41 = ScalarParameter() {defaultValue=-0.8, parameterName=MM2}
  n42 = Multiply(A=Saturate(None=Add(A=@n40, B=@n41)), B=@n34)
  n43 = Add(
   A=Multiply(A=Saturate(None=@n39), B=VectorParameter() {parameterName=HornGlowColor, defaultValue=<rgba 10 0.4365 9.5222 1>}.RGB)
   B=Multiply(A=Add(A=@n9, B=@n42), B=VectorParameter() {parameterName=HornGlowColorG, defaultValue=<rgba 3.1746 0.05 5 1>}.RGB)
  )
  n44 = LinearInterpolate(A=ScalarParameter() {parameterName=LeftSideGlow}, B=ScalarParameter() {parameterName=RightSideGlow}, Alpha=@n19)
  n45 = TextureSample() {texture=/Game/Characters/Enemies/DUDEBRO/Textures/T_DUDEBRO_Horns_D}
  n46 = VectorParameter() {parameterName=HornGlowDiffuseColor, defaultValue=<rgba 0.0257 0.0017 0.1 1>}
  n47 = Power(
   Base=Multiply(A=@n7, B=ScalarParameter() {defaultValue=1.2, parameterName=sdf3t33})
   Exp=ScalarParameter() {defaultValue=8, parameterName=6yhgrt444}
  )
  n48 = Multiply(A=Saturate(None=OneMinus(None=@n47)), B=@n44)
  n49 = ScalarParameter() {defaultValue=0.5, group=Base Material, parameterName=Specular Scale}
  n50 = MaterialFunctionCall!(
   InMA=MakeMaterialAttributes!(
    BaseColor=LinearInterpolate(A=@n45.RGB, B=@n46.RGB, Alpha=@n48)
    Metallic=@n29.G
    Specular=Multiply(A=@n49, B=@n29.R)
    Roughness=Add(A=@n35, B=@n36)
    Anisotropy
    EmissiveColor=Multiply(A=@n43, B=@n44)
    Opacity
    OpacityMask=Add(A=@n25, B=@n28)
    Normal=TextureSample() {samplerType=SAMPLERTYPE_Normal, texture=/Game/Characters/Enemies/DUDEBRO/Textures/T_DUDEBRO_Horns_N}.RGB
    Tangent
    WorldPositionOffset
    SubsurfaceColor
    ClearCoat
    ClearCoatRoughness
    AmbientOcclusion
    Refraction
    CustomizedUVs_0
    CustomizedUVs_1
    CustomizedUVs_2
    CustomizedUVs_3
    CustomizedUVs_4
    CustomizedUVs_5
    CustomizedUVs_6
    CustomizedUVs_7
    PixelDepthOffset
    ShadingModel
    Displacement
   )
   HaveMasks
   "FresnelRevert(WithOutMask)"
   EmissiveMutiply
  ) {materialFunction=/Game/Characters/Enemies/DUDEBRO/FX/SkinFX/MF_DudebroSkinFX}
  n51 = GetMaterialAttributes(None=@n50.MA) {attributeGetTypes [#1]}
  n52 = LinearInterpolate(A=@n19, B=OneMinus(None=@n19), Alpha=ScalarParameter() {defaultValue=0.5, parameterName=HitReactLR})
  StaticSwitchParameter(
   True=SetMaterialAttributes(MaterialAttributes=@n50.MA, Emissive Color=Multiply(A=@n51.Emissive Color, B=@n52)) {attributeSetTypes [#1]}
   False=@n50.MA
  ) {parameterName=BuffetBro}
 ]
}
